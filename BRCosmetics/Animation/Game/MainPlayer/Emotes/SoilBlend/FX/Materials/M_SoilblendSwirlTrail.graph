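# Translucent Niagara mesh-particle material for the swirl trail in the SoilBlend emote.
# SplineThicken sets noise-modulated ribbon width, then sine-driven offsets along vertex normals and Z add wobble.
# Opacity combines a panned, normal-map-distorted swirl texture with noise erosion, a dynamic reveal mask and sphere mask, skipping depth and view-angle fades on mobile.
# Particle color gets fake sun shading, with a sphere-masked point-light glow driven by dynamic parameters.
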Material /BRCosmetics/Animation/Game/MainPlayer/Emotes/SoilBlend/FX/Materials/M_SoilblendSwirlTrail
graph {
 properties {bAutomaticallySetUsageInEditor=false, bUsedWithNiagaraMeshParticles=true, blendMode=BLEND_Translucent, usedShadingModels=Default Lit}
 nodes [
  n1 = ParticleColor()
  n2 = ScalarParameter() {defaultValue=35, parameterName=RibbonTopWidth}
  n3 = TextureSampleParameter2D(UVs=Panner(Coordinate=TextureCoordinate() {vTiling=0}) {speedX=-0.25}) {parameterName=Texture_WPO, texture=/Game/Items/Traps/Textures/T_PerlinRGB_01_M}
  n4 = Saturate(None=@n3.R)
  n5 = ScalarParameter() {defaultValue=0.2, parameterName=WidthDifferenceStrenght}
  n6 = ScalarParameter() {defaultValue=35, parameterName=RibbonEndWidth}
  n7 = MaterialFunctionCall(
   WidthBase=LinearInterpolate(A=@n2, B=Multiply(A=@n2, B=@n4), Alpha=@n5)
   WidthTip=LinearInterpolate(A=@n2, B=Multiply(A=@n6, B=@n4), Alpha=@n5)
   Expand U or V UV Channel=StaticBool()
  ) {materialFunction=&SplineThicken}
  n8 = CollectionParameter() {
   collection=&FortniteMaterialParameters
   parameterName=SunlightDirectionalVector
  }
  n9 = LinearInterpolate(
   A=Multiply(A=@n1.RGB, B=ScalarParameter() {defaultValue=0.904, parameterName=Shadow Multiplier}) {constB=0.74}
   B=@n1.RGB
   Alpha=DotProduct(A=@n7.WorldPositionOffset, B=ComponentMask(None=@n8) {b=true, g=true, r=true})
  )
  n10 = ScalarParameter() {defaultValue=0.8, parameterName=GlobalShadowScale}
  n11 = VertexNormalWS()
  n12 = AtmosphericLightVector()
  n13 = NamedRerouteDeclaration(None=LinearInterpolate(A=Multiply(A=@n9, B=@n10) {constB=0.5}, B=@n9, Alpha=DotProduct(A=@n11, B=@n12))) {name=FinalColor, nodeColor=<rgba 0.4766 0 1 1>}
  n14 = MaterialFunctionCall(Emissive=@n13) {materialFunction=&MF_FX_EmissiveAdjustments}
  LinearInterpolate(A=@n14.Result, B=Multiply(A=@n13, B=NamedRerouteUsage()), Alpha=ScalarParameter() {parameterName=QualitySwitch})
  n15 = ComponentMask(None=TextureCoordinate()) {g=true}
  n16 = VectorParameter() {parameterName=Distort_SV, defaultValue=<rgba 1 0.25 -0.5 0>}
  n17 = DynamicParameter() {paramNames [Time, DistortTime, WPOSpeed, TimeOffsetRandom]}
  n18 = TextureSampleParameter2D(UVs=MaterialFunctionCall!(InUV, Texture, InRGB=@n16.RGB, InA=@n16.RGB, InTime=Add(A=@n17.DistortTime, B=@n17.DistortTime)) {materialFunction=/Game/Effects/Fort_Effects/Materials/_Functions/MF_TexturePanner}.UV) {parameterName=Texture_Distortion, samplerType=SAMPLERTYPE_Normal, texture=/Game/Effects/Fort_Effects/Textures/Tile/T_Wind_N}
  n19 = Multiply(A=Abs(None=Subtract(A=@n15) {constB=0.5}), B=@n18.RGB)
  n20 = ScalarParameter() {defaultValue=0.5, parameterName=Distort}
  NamedRerouteDeclaration(None=ComponentMask(None=Multiply(A=@n19, B=@n20)) {g=true, r=true}) {name=Distort, nodeColor=<rgba 0.875 0 1 1>}
  NamedRerouteDeclaration(None=@n18.RGB) {name=NormalOut, nodeColor=<rgba 0.7109 0 1 1>}
  n21 = NamedRerouteUsage()
  n22 = DynamicParameter() {parameterIndex=2, paramNames [StrengthPointLight, PointLightRadius, Param3, Param4]}
  n23 = Multiply(A=VectorParameter() {group=SphereMask, parameterName=SphereMask_Color, defaultValue=<rgba 0 1.1207 2 1>}.RGB, B=@n22.StrengthPointLight)
  n24 = WorldPosition()
  n25 = ActorPositionWS()
  n26 = ScalarParameter() {group=SphereMask, parameterName=Hardness}
  NamedRerouteDeclaration(None=LinearInterpolate(B=Multiply(A=@n21, B=@n23), Alpha=SphereMask(A=@n24.XYZ, B=@n25, Radius=@n22.PointLightRadius, Hardness=@n26))) {name=FinelEmissive, nodeColor=<rgba 0 0.0156 1 1>}
  n27 = ComponentMask(None=TextureCoordinate() {uTiling=2.2}) {r=true}
  n28 = ScalarParameter() {defaultValue=0.5, parameterName=WPOTile}
  n29 = Multiply(A=@n3.R) {constB=0.5}
  n30 = Add(A=Multiply(A=Time()) {constB=0.5}, B=DynamicParameter() {paramNames [Time, DistortTime, WPOSpeed, TimeOffsetRandom]}.TimeOffsetRandom)
  n31 = Sine(None=Add(A=Multiply(A=@n27, B=@n28) {constB=2}, B=Add(A=@n29, B=@n30)))
  n32 = TextureCoordinate() {uTiling=2.2}
  n33 = Time()
  n34 = DynamicParameter() {paramNames [Time, DistortTime, WPOSpeed, TimeOffsetRandom]}
  n35 = Add(
   A=Multiply(A=ComponentMask(None=@n32) {r=true}, B=ScalarParameter() {defaultValue=2, parameterName=SwirlOffsetTileMult}) {constB=2}
   B=Multiply(A=Add(A=@n33, B=@n34.WPOSpeed), B=ScalarParameter() {defaultValue=1, parameterName=SwirlTimeMult})
  )
  n36 = VertexNormalWS()
  n37 = ScalarParameter() {defaultValue=10, parameterName=SwirlOffsetStrenght}
  n38 = Add(
   A=Multiply(A=Constant3Vector() {constant=<rgba 0 0 1 0>}, B=ScalarParameter() {defaultValue=5, parameterName=RandomOffsetStrenght}) {constB=5}
   B=Multiply(A=Sine(None=@n35), B=Multiply(A=@n36, B=@n37) {constB=10})
  )
  Reroute(None=Add(A=@n7.WorldPositionOffset, B=Multiply(A=@n31, B=@n38)))
  Reroute(None=@n14.Result)
  n39 = VectorParameter() {parameterName=Texture_SV, defaultValue=<rgba 1 1 -1 0>}
  n40 = DynamicParameter() {paramNames [Time, DistortTime, WPOSpeed, TimeOffsetRandom]}
  n41 = Add(A=@n40.Time, B=@n40.Time)
  n42 = Add(
   A=MaterialFunctionCall(InRGB=@n39.RGB, InA=@n39.RGB, InTime=@n41) {materialFunction=&MF_TexturePanner}.UV
   B=NamedRerouteUsage()
  )
  n43 = LinearInterpolate(Alpha=DynamicParameter() {parameterIndex=1, paramNames [RevealMask,Param2,Param3,Param4]}.RevealMask) {constA=-1}
  n44 = Add(A=@n43)
  n45 = TextureCoordinate() {vTiling=0}
  n46 = OneMinus(None=ComponentMask(None=@n45) {r=true})
  n47 = Subtract(
   A=ChannelMaskParameter(None=TextureSampleParameter2D(UVs=@n42) {parameterName=TextureA, texture=/BRCosmetics/Athena/Effects/Particles/Skydiving/NS_Contrail_SoilBlend/T_Trail_Swirl}.RGBA) {maskChannel=Blue, parameterName=TextureMask, defaultValue=<rgba 0 0 1 0>}.R
   B=Multiply(A=SmoothStep(Min=@n43, Max=@n44, Value=@n46)) {constB=2}
  )
  n48 = VectorParameter() {parameterName=Erosion_SV, defaultValue=<rgba 1 1 0 -0.5>}
  n49 = TextureSampleParameter2D(UVs=MaterialFunctionCall(InRGB=@n48.RGB, InA=@n48.RGB, InTime=@n41) {materialFunction=&MF_TexturePanner}.UV) {parameterName=Texture_Erosion, texture=&T_PerlinRGB_01_M}
  n50 = ComponentMask(None=TextureCoordinate()) {g=true}
  n51 = OneMinus(None=Abs(None=Subtract(A=@n50) {constB=0.5}))
  n52 = Multiply(
   A=Subtract(A=Saturate(None=@n47), B=Multiply(A=@n49.R, B=@n51))
   B=SphereMask(A=@n45, B=Constant2Vector() {r=0.5}, Radius=Constant() {r=0.5}, Hardness=Constant() {r=0.5})
  )
  n53 = ScalarParameter() {defaultValue=0.25, parameterName=Texture_Subtract}
  n54 = Multiply(A=Saturate(None=Subtract(A=@n52, B=@n53) {constB=0.25}), B=@n1.A)
  n55 = Saturate(None=@n54)
  n56 = MaterialFunctionCall("FadeSensitivity (optional)"=ScalarParameter() {defaultValue=0.75, parameterName=ViewangleFadeSensitivity}) {materialFunction=&FadeBasedOnViewAngle}
  n57 = Multiply(
   A=DepthFade(FadeDistance=ScalarParameter() {defaultValue=10, parameterName=DepthFade}) {fadeDistanceDefault=10}
   B=LinearInterpolate(A=@n55, B=Multiply(A=@n55, B=@n56.Result), Alpha=ScalarParameter() {defaultValue=1, parameterName=ViewangleFadeStrenght})
  )
  Reroute(None=ShadingPathSwitch(Default=@n57, Deferred=@n57, Mobile=@n54) {desc=Migrated from FeatureLevelSwitch})
 ]
}
