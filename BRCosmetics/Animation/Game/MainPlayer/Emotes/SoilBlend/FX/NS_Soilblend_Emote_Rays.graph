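# Niagara system for the SoilBlend player emote that renders a green glow with flashing light rays.
# Glow is a camera-offset sprite burst whose custom ColorSet module combines a loop-age envelope with a multi-peak flicker curve.
# Ray_Soft bursts one mesh ray, pulsing its RGB with an age-driven waveform inside a fade curve.
# Rays_flash spawns short-lived, randomly scaled and tilted mesh rays that stretch while their alpha fades over the emitter loop.
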NiagaraSystem /BRCosmetics/Animation/Game/MainPlayer/Emotes/SoilBlend/FX/NS_Soilblend_Emote_Rays
dependencies {usedDataInterfaces [&NiagaraDataInterfaceCurve], usedDynamicInputs [], usedModules [&SystemState]}
emitters {
 Glow {
  emitterUpdateScript [
   &EmitterState(Life Cycle Mode=::NewEnumerator1, Loop Behavior=::NewEnumerator1, Loop Duration Mode=::NewEnumerator1, UseLoopDelay=false)
   &SpawnBurst_Instantaneous(Use Spawn Probability=false, UseLoopCountLimit=false)
  ]
  particleSpawnScript [
   &InitializeParticle(Color=<rgba 0 1 0.0049 1>, Sprite Size Mode=::NewEnumerator3, Uniform Sprite Size=100, UsePositionOffset=false, Write Material Random=true)
   &DynamicMaterialParameters(
    Index 0 Param 1=0.2
    Index 0 Param 2=1
    Module.Index0Param0Write=true
    Module.Index0Param1Write=true
    Module.Index0Param2Write=true
    Module.Index0Param3Write=true
   )
   &CameraOffset(Camera Offset Amount=30)
  ]
  particleUpdateScript [
   &ParticleState()
   &self:ColorSet(
    Linear Color:LinearColor=<rgba 0 1 0.136 1>
    Min=0.5
    Random:NiagaraFloat=&RandomRangeFloat(Minimum=0, Maximum=0.5, Evaluation Type=::NewEnumerator1)
    ScaleCurve:NiagaraFloat=&Multiply_Float(
     A=&FloatFromCurve(
      FloatCurve=DataInterface() {propertyValues {Curve {keys [
          {arriveTangent=9.0725, interpMode=RCIM_Cubic, leaveTangent=9.0725, tangentMode=RCTM_User, time=0, value=0}
          {arriveTangent=-2.5249, interpMode=RCIM_Cubic, leaveTangent=-2.5249, tangentMode=RCTM_User, time=0.1408, value=0.9847}
          {arriveTangent=-0.0405, interpMode=RCIM_Cubic, leaveTangent=-0.0405, tangentMode=RCTM_User, time=0.8, value=0}
         ]}}}
      CurveIndex=$Emitter.NormalizedLoopAge
     )
     B=&FloatFromCurve(FloatCurve=DataInterface() {propertyValues {Curve {keys [<at 0 0>,<at 0.1 0.5>,<at 0.2 0>,<at 0.4 0>,<at 0.5 0.75>,<at 0.6 0>,<at 0.7 0>,<at 0.8 0.25>,<at 0.9 0>,<at 1 0>]}}})
    )
   ) {module=ColorSet}
  ]
  rendererClasses [&NiagaraSpriteRendererProperties]
  renderers [{rendererClass=&NiagaraSpriteRendererProperties}]
 }
 Ray_Soft {
  emitterUpdateScript [
   &EmitterState(Life Cycle Mode=::NewEnumerator1, Loop Behavior=::NewEnumerator1, Loop Duration Mode=::NewEnumerator1, UseLoopDelay=false)
   &SpawnBurst_Instantaneous(Spawn Count=1, Use Spawn Probability=false)
  ]
  particleSpawnScript [
   &InitializeParticle(Color=<rgba 0 1 0.1443 0.6>, Mesh Scale=<1 0.4 0.4>, Mesh Scale Mode=::NewEnumerator4, UsePositionOffset=false, Write Material Random=true)
   &SphereLocation(
    MaskSpawnGroup=false
    Override Seed=false
    Sphere Radius=0
    Surface Only=false
    UseNonUniformScale=false
    UseOffset=true
    UseSphereOrientation=false
   )
   &InitialMeshOrientation(Use Orientation Vector=true, Use Rotation Vector=true)
  ]
  particleUpdateScript [
   &ParticleState(Kill Particles When Lifetime Has Elapsed=false, Let Infinitely Lived Particles Die When Emitter Deactivates=true)
   &SolveForcesAndVelocity(Clamp Velocity=false, Limit Acceleration=false)
   &ScaleColor(
    Scale RGB=&VectorFromFloat(Value=&FloatFromCurve(
      FloatCurve=DataInterface() {propertyValues {Curve {keys [
          {arriveTangent=1.4025, interpMode=RCIM_Cubic, leaveTangent=1.4025, tangentMode=RCTM_User, time=0, value=0}
          <at 0.7 1 RCIM_Cubic>
          <at 1 0 RCIM_Cubic>
         ]}}}
      Scale Curve=&Lerp_Float(
       A=1
       B=8
       Alpha=&Waveform("Waveform [1]"=::NewEnumerator0, "[1] Frequency"=5, "[1] Amplitude Min/Max"=<0.5 1>, "[2] Phase"=$Emitter.Age, "[3] Phase"=$Emitter.Age)
      )
     ))
    ScaleA=true
    ScaleRGB=true
   )
  ]
  rendererClasses [&NiagaraMeshRendererProperties]
  renderers [{rendererClass=&NiagaraMeshRendererProperties, rendererIndex=0}]
 }
 Rays_flash {
  emitterUpdateScript [
   &EmitterState(Life Cycle Mode=::NewEnumerator1, Loop Behavior=::NewEnumerator1, Loop Duration=0.5, UseLoopDelay=false)
   &SpawnRate(SpawnRate=20, Use Spawn Probability=false)
  ]
  particleSpawnScript [
   &InitializeParticle(
    Color=&RandomRangeLinearColor(Minimum=<rgba 0 1 0.0693 4>, Maximum=<rgba 0 1 0.1384 2>)
    Lifetime Max=0.5
    Lifetime Min=0.2
    Lifetime Mode=::NewEnumerator1
    Mesh Scale Max=<1.7 1 0.3>
    Mesh Scale Min=<1.2 1 0.1>
    Mesh Scale Mode=::NewEnumerator2
    UsePositionOffset=false
    Write Material Random=true
   ) {module=InitializeParticle001}
   &SphereLocation(
    MaskSpawnGroup=false
    Override Seed=false
    Sphere Radius=0
    Surface Only=false
    UseNonUniformScale=false
    UseOffset=true
    UseSphereOrientation=false
   )
   &DynamicMaterialParameters(
    Index 0 Param 1=&UniformRangedFloat(Minimum=-1, Maximum=1)
    Index 0 Param 2=1
    Module.Param0WriteEnabled=true
    Module.Param1WriteEnabled=true
    Module.Param2WriteEnabled=true
    Module.Param3WriteEnabled=true
   )
   &InitialMeshOrientation(
    Rotation=&RandomRangeVector(Minimum=<0.025 0.025 0.025>, Maximum=<-0.025 -0.025 -0.025>)
    Use Orientation Vector=true
    Use Rotation Vector=true
   ) {module=InitialMeshOrientation001}
  ]
  particleUpdateScript [
   &ParticleState()
   &ScaleColor(
    Scale Alpha=&FloatFromCurve(
     FloatCurve=DataInterface() {propertyValues {Curve {keys [
         {interpMode=RCIM_Cubic, tangentMode=RCTM_User, time=0, value=0}
         {interpMode=RCIM_Cubic, tangentMode=RCTM_User, time=0.3, value=0.5}
         {tangentMode=RCTM_User, time=1, value=0}
        ]}}}
     Scale Curve=&FloatFromCurve(
      FloatCurve=DataInterface() {propertyValues {Curve {keys [<at 0 1 RCIM_Cubic>, {arriveTangent=-2.4739, interpMode=RCIM_Cubic, tangentMode=RCTM_Break, time=1, value=0}]}}}
      CurveIndex=$Emitter.NormalizedLoopAge
     )
    )
    ScaleA=true
    ScaleRGB=false
   ) {module=ScaleColor001}
   &ScaleMeshSize(Scale Factor=&MakeVector(
     X=&FloatFromCurve(FloatCurve=DataInterface() {propertyValues {Curve {keys [<at 0 0.8>,<at 1 1.2>]}}})
     Y=1
     Z=&FloatFromCurve(FloatCurve=DataInterface() {propertyValues {Curve {keys [
          {interpMode=RCIM_Cubic, tangentMode=RCTM_Break, time=0, value=0}
          {interpMode=RCIM_Cubic, tangentMode=RCTM_Break, time=0.15, value=1}
          {interpMode=RCIM_Cubic, tangentMode=RCTM_Break, time=0.85, value=1}
          <at 1 0>
         ]}}})
    ))
  ]
  rendererClasses [&NiagaraMeshRendererProperties]
  renderers [{rendererClass=&NiagaraMeshRendererProperties, rendererIndex=0}]
 }
}
summary {
 systemName=NS_Soilblend_Emote_Rays
 emitters [
  {emitterName=Ray_Soft, rendererClasses [&NiagaraMeshRendererProperties]}
  {emitterName=Rays_flash, rendererClasses [&NiagaraMeshRendererProperties]}
  {emitterName=Glow, rendererClasses [&NiagaraSpriteRendererProperties]}
 ]
 userVariables []
}
