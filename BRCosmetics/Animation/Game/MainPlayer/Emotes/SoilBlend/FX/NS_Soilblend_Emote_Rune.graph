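# Niagara effect for the SoilBlend emote: green spinning runes with a flickering glow, particle light and wispy smoke.
# Rune001 bursts eight sprites whose custom rotation module reads size, spin, intensity and SubUV arrays, with spin speed driving color and material.
# Glow samples skeletal-mesh sockets, smoke partly inherits owner movement by speed, and a system StartupCurve scales glow size and rune color.
# Older Rune and Rune_Low emitters are disabled.
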NiagaraSystem /BRCosmetics/Animation/Game/MainPlayer/Emotes/SoilBlend/FX/NS_Soilblend_Emote_Rune
dependencies {
 usedDataInterfaces [
  &NiagaraDataInterfaceCurve
  &NiagaraDataInterfaceVector2DCurve
  &NiagaraDataInterfaceArrayFloat
  &NiagaraDataInterfaceSpriteRendererInfo
 ]
 usedDynamicInputs [&Vector2DFromCurve]
 usedModules [
  &self:SystemSpawnScript.SystemScriptSource.SystemScriptGraph.NiagaraNodeAssignment_0.SetVariables_#1
  &SystemState
  &self:SystemSpawnScript.SystemScriptSource.SystemScriptGraph.NiagaraNodeAssignment_1.SetVariables_#2
 ]
}
emitters {
 Glow {
  emitterUpdateScript [
   &EmitterState(
    Inactive Response=::NewEnumerator2
    Life Cycle Mode=::NewEnumerator1
    Loop Behavior=::NewEnumerator1
    Loop Delay=0.1
    Loop Duration Mode=::NewEnumerator1
    UseLoopDelay=true
   )
   &SpawnBurst_Instantaneous(Loop Count Limit=0, Use Spawn Probability=false, UseLoopCountLimit=false)
  ]
  particleSpawnScript [
   &InitializeParticle(
    Color Mode=::NewEnumerator0
    Lifetime=&RandomRangeFloat(Minimum=1, Maximum=2)
    Position Mode=::NewEnumerator1
    Sprite Size=<111 33>
    Sprite Size Mode=::NewEnumerator4
    UsePositionOffset=false
    Write Material Random=true
   )
   &DynamicMaterialParameters(
    Index 0 Param 1=0.2
    Index 0 Param 2=1
    Module.Param0WriteEnabled=true
    Module.Param1WriteEnabled=true
    Module.Param2WriteEnabled=true
    Module.Param3WriteEnabled=true
   )
  ]
  particleUpdateScript [
   &SampleSkeletalMesh(
    AdditionalYawPitchRoll=false
    Bone / Socket Index:NiagaraInt32=$Particles.SocketIndex
    Bone Sampling Mode:ENiagaraBoneSamplingMode=::NewEnumerator4
    Position Offset Coordinate Space:ENiagaraCoordinateSpace=::World
    Sampled Position Offset:Vector3f=<0 12 0>
    Skeletal Mesh:NiagaraDataInterfaceSkeletalMesh=$System.Skeletal Mesh
   )
   &ParticleState(
    Kill Particles When Lifetime Has Elapsed=false
    Let Infinitely Lived Particles Die When Emitter Deactivates=true
    Loop Particles Lifetime=true
   )
   &ScaleSpriteSize(Scale Sprite Size Mode=::NewEnumerator0, Uniform Scale Factor=&Multiply_Float(A=1, B=$System.StartupCurve))
   &self:ColorSet_01(
    ColorScale:NiagaraFloat=&Lerp_Float(
     A=0
     B=1
     Alpha=&FloatFromCurve(
      FloatCurve=DataInterface() {propertyValues {Curve {keys [{tangentMode=RCTM_User, time=0, value=0}, {tangentMode=RCTM_User, time=0.2, value=1}]}}}
      CurveIndex=$Engine.System.Age
     )
    )
    Linear Color:LinearColor=<rgba 0 1 0.1294 1>
    Min=0.5
    Random:NiagaraFloat=&RandomRangeFloat(Minimum=0, Maximum=0.1, Evaluation Type=::NewEnumerator1)
    ScaleCurve:NiagaraFloat=&FloatFromCurve(
     FloatCurve=DataInterface() {propertyValues {Curve {keys [<at 0 0>,<at 0.1 0.5>,<at 0.2 0>,<at 0.4 0>,<at 0.5 0.75>,<at 0.6 0>,<at 0.7 0>,<at 0.8 0.25>,<at 0.9 0>,<at 1 0>]}}}
     Scale Curve=1.5
    )
   ) {module=ColorSet_01}
   &FN_Light_Attributes(
    Color=$Particles.Color
    Create Color Attribute=true
    Create Exponent Attribute=true
    Create Radius Attribute=true
    Enable Particle Light Enable=true
    Enable Visibility Tag=true
    Exponent=2
    Light Enabled=true
    Radius=50.5276
    Scale Color=&Multiply_Float(
     A=20
     B=&FloatFromCurve(
      FloatCurve=DataInterface() {propertyValues {Curve {keys [{tangentMode=RCTM_User, time=0, value=0}, {tangentMode=RCTM_User, time=0.25, value=1}]}}}
      CurveIndex=$Emitter.Age
     )
    )
    VisibilityTag=1
   )
   &CameraOffset(Camera Offset Amount=59)
  ]
  rendererClasses [&NiagaraSpriteRendererProperties, &NiagaraLightRendererProperties]
  renderers [
   {rendererClass=&NiagaraSpriteRendererProperties}
   {rendererClass=&NiagaraLightRendererProperties, rendererIndex=1}
  ]
 }
 Rune {
  bEnabled=false
  emitterUpdateScript [
   &EmitterState(Life Cycle Mode=::NewEnumerator1, Loop Behavior=::NewEnumerator1, Loop Delay=0.1, Loop Duration Mode=::NewEnumerator1, UseLoopDelay=true)
   &SpawnBurst_Instantaneous(Spawn Count=8, Use Spawn Probability=false, UseLoopCountLimit=false)
  ]
  particleSpawnScript [
   /SetVariables(Particles.SocketIndex=HlslExpression() {hlslExpression="fmod(Engine.ExecIndex,2)"}) {isSetParameters=true, module=SetVariables_#3}
   &InitializeParticle(
    Color=<rgba 0.0425 0.5156 0 1>
    Sprite Size Mode=::NewEnumerator3
    Uniform Sprite Size=35
    UsePositionOffset=false
    Write Material Random=true
   )
  ]
  particleUpdateScript [
   &ParticleState(Kill Particles When Lifetime Has Elapsed=false, Let Infinitely Lived Particles Die When Emitter Deactivates=true)
   &ScaleSpriteSize(
    Uniform Curve Index=$Engine.System.Age
    Uniform Curve Sprite Scale=DataInterface() {propertyValues {Curve {keys [
        {interpMode=RCIM_Cubic, tangentMode=RCTM_User, time=0, value=0}
        {interpMode=RCIM_Cubic, tangentMode=RCTM_User, time=0.5, value=1}
        {tangentMode=RCTM_User, time=1, value=0}
       ]}}}
   )
   &self:RuneRotation_01(
    Intensity=DataInterface() {propertyValues {GpuSyncMode=SyncCpuToGpu, MaxElements=0, FloatData [0.3,0.3,0.6,0.6,1.2,1.2,0.85,0.85]}}
    MaxSpinDegree=360
    Size=DataInterface() {propertyValues {GpuSyncMode=SyncCpuToGpu, MaxElements=0, FloatData [1.5,1.5,1.2,1.2,0.2,0.2,0.6,0.6]}}
    SizeMult:NiagaraFloat=$System.StartupCurve
    Spin=DataInterface() {propertyValues {GpuSyncMode=SyncCpuToGpu, MaxElements=0, FloatData [1,1,3,3,0,0,0,0]}}
    SubuvIndex=DataInterface() {propertyValues {GpuSyncMode=SyncCpuToGpu, MaxElements=0, FloatData [1,1,0,0,2,2,3,3]}}
    Wiggle=1
   ) {module=RuneRotation_01}
   &ScaleColor(
    Color Value To Scale=$Particles.Color
    Scale Mode=::NewEnumerator1
    Scale RGBA=&MakeVector4_FromLinearColor(LinearColor=&MultiplyLinearColor(
      Linear Color=/Niagara/DynamicInputs/Lerp/SmoothLerpOverTime_Color(
       Smooth Value:LinearColor=<rgba 0.0388 0.4566 0.1249 2>
       Target Value:LinearColor=<rgba 0 10 0.6367 0.5>
       Convergence Rate:NiagaraFloat=&Abs_Float(Float:NiagaraFloat=&Multiply_Float(A=$Particles.SpinSpeed, B=50))
      )
      Scale Factor=&MakeVector4_FromFloat(Float=$System.StartupCurve)
     ))
    ScaleRGBA=true
   )
   &CameraOffset(Camera Offset Amount=&RandomRangeFloat(Minimum=15, Maximum=25, Evaluation Type=::NewEnumerator1))
   &DynamicMaterialParameters(
    Index 0 Param 1=&Multiply_Float(A=$Particles.SpinSpeed, B=8)
    Module.Index0Param0Write=true
    Module.Index0Param1Write=true
    Module.Index0Param2Write=true
    Module.Index0Param3Write=true
   )
   &PartitionParticles()
  ]
  rendererClasses [&NiagaraSpriteRendererProperties]
  renderers [
   {rendererClass=&NiagaraSpriteRendererProperties}
   {rendererClass=&NiagaraSpriteRendererProperties, rendererIndex=1}
  ]
 }
 Rune001 {
  emitterUpdateScript [
   &EmitterState(
    Inactive Response=::NewEnumerator2
    Life Cycle Mode=::NewEnumerator1
    Loop Behavior=::NewEnumerator1
    Loop Duration Mode=::NewEnumerator1
    UseLoopDelay=false
   )
   &SpawnRate(SpawnRate=2, Use Spawn Probability=false) {enabled=false}
   &SpawnBurst_Instantaneous(Spawn Count=8, Use Spawn Probability=false, UseLoopCountLimit=false)
  ]
  particleSpawnScript [
   /SetVariables(Particles.SocketIndex=HlslExpression() {hlslExpression="fmod(Engine.ExecIndex,2)"}) {enabled=false, isSetParameters=true, module=SetVariables_#3}
   &InitializeParticle(Color=<rgba 2 2 2 1>, Sprite Size Mode=::NewEnumerator3, Uniform Sprite Size=35, UsePositionOffset=false, Write Material Random=true)
  ]
  particleUpdateScript [
   /SetVariables(Particles.Position=$StackContext.SampleSkeletalMesh.SampledPosition) {enabled=false, isSetParameters=true}
   &ParticleState(Kill Particles When Lifetime Has Elapsed=false, Let Infinitely Lived Particles Die When Emitter Deactivates=true)
   &ScaleSpriteSize(
    Uniform Curve Index=$Engine.System.Age
    Uniform Curve Sprite Scale=DataInterface() {propertyValues {Curve {keys [{interpMode=RCIM_Cubic, tangentMode=RCTM_User, time=0, value=0}, {interpMode=RCIM_Cubic, tangentMode=RCTM_User, time=0.5, value=1}]}}}
   ) {module=ScaleSpriteSize002}
   &ScaleSpriteSize(Uniform Curve Sprite Scale=DataInterface() {propertyValues {Curve {keys [{tangentMode=RCTM_User, time=0, value=0.75}, {tangentMode=RCTM_User, time=1, value=1}]}}}) {enabled=false}
   &self:RuneRotation_02(
    Intensity=DataInterface() {propertyValues {GpuSyncMode=SyncCpuToGpu, MaxElements=0, FloatData [0.3,0.3,0.6,0.6,1.2,1.2,0.85,0.85]}}
    MaxSpinDegree=360
    Size=DataInterface() {propertyValues {GpuSyncMode=SyncCpuToGpu, MaxElements=0, FloatData [1.5,1.5,1.2,1.2,0.2,0.2,0.6,0.6]}}
    Spin=DataInterface() {propertyValues {GpuSyncMode=SyncCpuToGpu, MaxElements=0, FloatData [1,1,3,3,0,0,0,0]}}
    SubuvIndex=DataInterface() {propertyValues {GpuSyncMode=SyncCpuToGpu, MaxElements=0, FloatData [1,1,0,0,2,2,3,3]}}
    Wiggle=1
   ) {module=RuneRotation_02}
   &ScaleColor(
    Color Value To Scale=$Particles.Color
    Scale Mode=::NewEnumerator1
    Scale RGBA=&MakeVector4_FromLinearColor(LinearColor=&SmoothLerpOverTime_Color(
      Smooth Value:LinearColor=<rgba 0.0388 0.4566 0.1249 2>
      Target Value:LinearColor=<rgba 0 10 0.6367 0.5>
      Convergence Rate:NiagaraFloat=&Abs_Float(Float:NiagaraFloat=&Multiply_Float(A=$Particles.SpinSpeed, B=50))
     ))
    ScaleRGBA=true
   ) {enabled=false}
   &ScaleColor(
    Color Value To Scale=$Particles.Color
    Scale Mode=::NewEnumerator1
    Scale RGBA=&MakeVector4_FromLinearColor(LinearColor=&MultiplyLinearColor(
      Linear Color=&SmoothLerpOverTime_Color(
       Smooth Value:LinearColor=<rgba 0.0388 0.4566 0.1249 2>
       Target Value:LinearColor=<rgba 0 10 0.6367 0.5>
       Convergence Rate:NiagaraFloat=&Abs_Float(Float:NiagaraFloat=&Multiply_Float(A=$Particles.SpinSpeed, B=50))
      )
      Scale Factor=&MakeVector4_FromFloat(Float=$System.StartupCurve)
     ))
    ScaleRGBA=true
   ) {module=ScaleColor001}
   &CameraOffset(Camera Offset Amount=&RandomRangeFloat(Minimum=15, Maximum=25, Evaluation Type=::NewEnumerator1))
   &DynamicMaterialParameters(
    Index 0 Param 1=&Multiply_Float(A=$Particles.SpinSpeed, B=8)
    Module.Index0Param0Write=true
    Module.Index0Param1Write=true
    Module.Index0Param2Write=true
    Module.Index0Param3Write=true
   )
   &JitterPosition(Jitter Amount=1, Jitter Delay=0.1) {enabled=false}
   &ScaleSpriteSize(
    Initial Sprite Size:Vector2f=$Particles.SpriteSize
    Scale Sprite Size Mode=::NewEnumerator0
    Uniform Scale Factor=&Add_Float(
     A=1
     B=&Multiply_Float(
      A=&VectorLength(Vector=$Engine.Owner.Velocity)
      B=&Add_Float(
       A=&Multiply_Float(A=0, B=&RandomRangeFloat(Minimum=-1, Maximum=1, Evaluation Type=::NewEnumerator1))
       B=0
      )
     )
    )
   ) {module=ScaleSpriteSize001}
  ]
  rendererClasses [&NiagaraSpriteRendererProperties]
  renderers [{rendererClass=&NiagaraSpriteRendererProperties}]
 }
 Rune_Low {
  bEnabled=false
  emitterUpdateScript [
   &EmitterState(
    Inactive Response=::NewEnumerator1
    Life Cycle Mode=::NewEnumerator1
    Loop Behavior=::NewEnumerator1
    Loop Duration Mode=::NewEnumerator1
    UseLoopDelay=false
   )
   &SpawnBurst_Instantaneous(Spawn Count=4, Use Spawn Probability=false, UseLoopCountLimit=false)
  ]
  particleSpawnScript [
   /SetVariables(Particles.SocketIndex=HlslExpression() {hlslExpression="fmod(Engine.ExecIndex,2)"}) {isSetParameters=true, module=SetVariables_#3}
   &InitializeParticle(
    Color=<rgba 0.0425 0.5156 0 1>
    Lifetime=3
    Sprite Size Mode=::NewEnumerator3
    Uniform Sprite Size=35
    UsePositionOffset=false
    Write Material Random=true
   )
  ]
  particleUpdateScript [
   &ParticleState(Kill Particles When Lifetime Has Elapsed=false, Let Infinitely Lived Particles Die When Emitter Deactivates=true)
   &ScaleSpriteSize(
    Uniform Curve Index=$Engine.System.Age
    Uniform Curve Sprite Scale=DataInterface() {propertyValues {Curve {keys [
        {interpMode=RCIM_Cubic, tangentMode=RCTM_User, time=0, value=0}
        {interpMode=RCIM_Cubic, tangentMode=RCTM_User, time=0.5, value=1}
        {tangentMode=RCTM_User, time=1, value=0}
       ]}}}
   )
   &self:RuneRotation_01(
    Intensity=DataInterface() {propertyValues {GpuSyncMode=SyncCpuToGpu, MaxElements=0, FloatData [0.3,0.6,1.2,0.85]}}
    MaxSpinDegree=360
    Size=DataInterface() {propertyValues {GpuSyncMode=SyncCpuToGpu, MaxElements=0, FloatData [1.5,1.2,0.2,0.6]}}
    SizeMult:NiagaraFloat=$System.StartupCurve
    Spin=DataInterface() {propertyValues {GpuSyncMode=SyncCpuToGpu, MaxElements=0, FloatData [1,3,0,0]}}
    SubuvIndex=DataInterface() {propertyValues {GpuSyncMode=SyncCpuToGpu, MaxElements=0, FloatData [1,0,2,3]}}
    Wiggle=1
   ) {module=RuneRotation_01}
   &ScaleColor(
    Color Value To Scale=$Particles.Color
    Scale Mode=::NewEnumerator1
    Scale RGBA=&MakeVector4_FromLinearColor(LinearColor=&MultiplyLinearColor(
      Linear Color=&SmoothLerpOverTime_Color(
       Smooth Value:LinearColor=<rgba 0.0388 0.4566 0.1249 2>
       Target Value:LinearColor=<rgba 0 10 0.6367 0.5>
       Convergence Rate:NiagaraFloat=&Abs_Float(Float:NiagaraFloat=&Multiply_Float(A=$Particles.SpinSpeed, B=50))
      )
      Scale Factor=&MakeVector4_FromFloat(Float=$System.StartupCurve)
     ))
    ScaleRGBA=true
   )
   &CameraOffset(Camera Offset Amount=&RandomRangeFloat(Minimum=15, Maximum=25, Evaluation Type=::NewEnumerator1))
   &DynamicMaterialParameters(
    Index 0 Param 1=&Multiply_Float(A=$Particles.SpinSpeed, B=8)
    Module.Index0Param0Write=true
    Module.Index0Param1Write=true
    Module.Index0Param2Write=true
    Module.Index0Param3Write=true
   )
  ]
  rendererClasses [&NiagaraSpriteRendererProperties]
  renderers [{rendererClass=&NiagaraSpriteRendererProperties}]
 }
 WispySmoke {
  emitterUpdateScript [
   &EmitterState(Life Cycle Mode=::NewEnumerator1, Loop Duration=1, UseLoopDelay=false)
   &SpawnRate(SpawnRate=8, Use Spawn Probability=false)
  ]
  particleSpawnScript [
   &InitializeParticle(
    Color=<rgba 0 0.8385 0.2096 1>
    Sprite Rotation Mode=::NewEnumerator2
    Sprite Size Mode=::NewEnumerator3
    Uniform Sprite Size=66
    UsePositionOffset=false
    Write Material Random=true
   )
   &ShapeLocation(
    MaskSpawnGroup=false
    Override Seed=false
    Rotation Angle=180
    Rotation Axis=<0 0 1>
    Rotation Mode=::NewEnumerator1
    Sphere Radius=7
    UseOwnerScale=true
   )
   &AddVelocity(Module.Use Falloff Distance=false, Override Seed=false, Velocity Mode=::NewEnumerator1, Velocity Speed=20)
  ]
  particleUpdateScript [
   &Drag(Use Linear Drag=true, Use Rotational Drag=true)
   &GravityForce(Gravity=<0 0 -50>)
   &InheritSourceMovement(
    Applied Position Scale=&Multiply_VectorByFloat(
     Vector=<1 1 1>
     Float=&Lerp_Float(
      A=&FloatFromCurve(FloatCurve=DataInterface() {propertyValues {Curve {keys [{arriveTangent=-1.0658, interpMode=RCIM_Cubic, leaveTangent=-1.0658, tangentMode=RCTM_User, time=0.1, value=1}, <at 1 0.4 RCIM_Cubic>]}}})
      B=&FloatFromCurve(FloatCurve=DataInterface() {propertyValues {Curve {keys [{arriveTangent=-0.0965, interpMode=RCIM_Cubic, leaveTangent=-0.0965, tangentMode=RCTM_User, time=0.3, value=1}, <at 1 0.7 RCIM_Cubic>]}}})
      Alpha=&MakeCustomFloatFromBool(
       Bool=&SetBoolByFloatComparison(A=&VectorLength(Vector=$Engine.Owner.Velocity), B=400)
       True Float=1
       False Float=0
      )
     )
    )
    ApplyForce=false
    ApplyPosition=true
    ApplyPositionFalloff=false
   )
   &SolveForcesAndVelocity(Clamp Velocity=false, Limit Acceleration=false)
   &ParticleState()
   &ScaleSpriteSize(Uniform Curve Sprite Scale=DataInterface() {propertyValues {Curve {keys [{tangentMode=RCTM_User, time=0, value=0.5}, {tangentMode=RCTM_User, time=1, value=1}]}}})
   &SubUVAnimation(
    End Frame Range Override=32
    "Sprite Renderer "=DataInterface() {propertyValues {SpriteRenderer=&self:WispySmoke_2.NiagaraSpriteRendererProperties_0}}
    Start Frame Range Override=16
    UseEndFrame=true
    UseStartFrame=true
   )
   &ScaleColor(
    Scale Alpha=&FloatFromCurve(
     FloatCurve=DataInterface() {propertyValues {Curve {keys [
         {interpMode=RCIM_Cubic, tangentMode=RCTM_User, time=0, value=0}
         {interpMode=RCIM_Cubic, tangentMode=RCTM_User, time=0.5, value=1}
         {tangentMode=RCTM_User, time=1, value=0}
        ]}}}
     Scale Curve=0.25
    )
    ScaleA=true
    ScaleRGB=true
   )
  ]
  rendererClasses [&NiagaraSpriteRendererProperties]
  renderers [{rendererClass=&NiagaraSpriteRendererProperties}]
 }
}
summary {
 systemName=NS_Soilblend_Emote_Rune
 emitters [
  {bEnabled=false, emitterName=Rune, rendererClasses [&NiagaraSpriteRendererProperties]}
  {emitterName=WispySmoke, rendererClasses [&NiagaraSpriteRendererProperties]}
  {emitterName=Glow, rendererClasses [&NiagaraSpriteRendererProperties, &NiagaraLightRendererProperties]}
  {bEnabled=false, emitterName=Rune_Low, rendererClasses [&NiagaraSpriteRendererProperties]}
  {emitterName=Rune001, rendererClasses [&NiagaraSpriteRendererProperties]}
 ]
 userVariables []
}
