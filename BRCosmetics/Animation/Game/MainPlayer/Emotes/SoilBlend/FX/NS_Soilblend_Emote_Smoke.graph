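# Niagara system for the SoilBlend emote, producing swirling green smoke with spiral and swirl meshes and a glow light.
# Smoke001 and WispySmoke spawn sprites on cylinders, spin them with vortex forces and inherit owner movement, scaled by speed.
# Spiral and Swirl burst partitioned meshes with time-scrolled material parameters; Spiral adds the light, Swirl's count follows quality level.
# System timelines drive fades, radius and vortex strength, while time-of-day and front-end modifiers adjust color and brightness.
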NiagaraSystem /BRCosmetics/Animation/Game/MainPlayer/Emotes/SoilBlend/FX/NS_Soilblend_Emote_Smoke
dependencies {
 usedDataInterfaces [
  &NiagaraDataInterfaceArrayInt32
  &NiagaraDataInterfaceCurve
  &NiagaraDataInterfaceSpriteRendererInfo
  &NiagaraDataInterfaceArrayFloat
  &NiagaraDataInterfaceCamera
  &NiagaraDataInterfaceColorCurve
 ]
 usedDynamicInputs [&Multiply_Vector, &ReturnCameraVector]
 usedModules [
  &SystemState
  &Timeline
  /BRCosmetics/Characters/Player/Female/Medium/Bodies/F_MED_PiperShelf_Pearl/FX/TimeOfDayGlobal
 ]
}
emitters {
 Smoke001 {
  emitterSpawnScript [
   /SetVariables(Emitter.FakeLightActive=true) {isSetParameters=true}
   /SetVariables(Emitter.FrontEndShadowLow:NiagaraFloat=&FN_FrontEndFloatModifier(Gameplay=0.7, Front End=1.5)) {isSetParameters=true}
  ]
  emitterUpdateScript [
   &EmitterState()
   &SpawnRate(SpawnRate=16, Use Spawn Probability=true)
  ]
  particleSpawnScript [
   &InitializeParticle(
    Color=<rgba 0.4 0.4 0.4 1>
    Lifetime Max=1.75
    Lifetime Min=1.5
    Lifetime Mode=::NewEnumerator1
    Sprite Rotation Angle=&RandomRangeFloat(Minimum=0.3, Maximum=-0.3)
    Sprite Rotation Mode=::NewEnumerator1
    Sprite Size Mode=::NewEnumerator1
    Uniform Sprite Size Max=85
    Uniform Sprite Size Min=75
    UsePositionOffset=true
    Write Material Random=true
   )
   &ShapeLocation(
    Cylinder Distribution=::NewEnumerator2
    Cylinder Height=5
    Cylinder Radius=&Lerp_Float(A=35, B=40, Alpha=$System.Timeline002.Float)
    MaskSpawnGroup=false
    Shape Primitive=::NewEnumerator1
    U Position=&RandomRangeFloat(Minimum=0, Maximum=1)
    UseOwnerScale=true
   )
   &CameraOffset(Camera Offset Amount=75)
   &ApplyInitialForces()
  ]
  particleUpdateScript [
   &InheritSourceMovement(
    Applied Position Scale=&Multiply_VectorByFloat(
     Vector=<1 1 1>
     Float=&Lerp_Float(
      A=&FloatFromCurve(FloatCurve=DataInterface() {propertyValues {Curve {keys [{arriveTangent=-1.0658, interpMode=RCIM_Cubic, leaveTangent=-1.0658, tangentMode=RCTM_User, time=0.1, value=1}, <at 1 0.4 RCIM_Cubic>]}}})
      B=&FloatFromCurve(FloatCurve=DataInterface() {propertyValues {Curve {keys [{arriveTangent=-0.0965, interpMode=RCIM_Cubic, leaveTangent=-0.0965, tangentMode=RCTM_User, time=0.3, value=1}, <at 1 0.7 RCIM_Cubic>]}}})
      Alpha=&MakeCustomFloatFromBool(
       Bool=&SetBoolByFloatComparison(A=&VectorLength(Vector=$Engine.Owner.Velocity), B=400)
       True Float=1
       False Float=0
      )
     )
    )
    ApplyForce=false
    ApplyPosition=true
    ApplyPositionFalloff=false
   )
   &DynamicMaterialParameters(
    Index 0 Param 1=&Add_Float(
     A=$Particles.RandomFloat
     B=&FloatFromCurve(FloatCurve=DataInterface() {propertyValues {Curve {keys [
          {arriveTangent=0.1552, interpMode=RCIM_Cubic, leaveTangent=0.1552, tangentMode=RCTM_User, time=0, value=1}
          {interpMode=RCIM_Cubic, leaveTangent=0.5473, time=1, value=1.1}
         ]}}})
    )
    Index 0 Param 4=1
    Index 2 Param 1=&FloatFromCurve(FloatCurve=DataInterface() {propertyValues {Curve {keys [
         {interpMode=RCIM_Cubic, leaveTangent=1.6214, tangentMode=RCTM_Break, time=0.1, value=0}
         {arriveTangent=0.8413, tangentMode=RCTM_User, time=1, value=1}
        ]}}})
    Index 2 Param 2=&FloatFromCurve(
     FloatCurve=DataInterface() {propertyValues {Curve {keys [{arriveTangent=0.7845, interpMode=RCIM_Cubic, leaveTangent=0.7845, tangentMode=RCTM_User, time=0, value=0.2}, <at 1 0.8 RCIM_Cubic>]}}}
     Scale Curve=&RandomRangeFloat(Minimum=0.2, Maximum=1.5)
    )
    Index 2 Param 3=&FloatFromCurve(
     FloatCurve=DataInterface() {propertyValues {Curve {keys [
         {interpMode=RCIM_Cubic, leaveTangent=1.0458, tangentMode=RCTM_Break, time=0, value=1}
         {interpMode=RCIM_Cubic, leaveTangent=11.1233, tangentMode=RCTM_Break, time=1, value=2}
        ]}}}
     Scale Curve=1.25
    )
    Index 2 Param 4=&FloatFromCurve(
     FloatCurve=DataInterface() {propertyValues {Curve {keys [<at 0 1 RCIM_Cubic>, {interpMode=RCIM_Cubic, leaveTangent=-1.2956, tangentMode=RCTM_Break, time=1, value=0.75}]}}}
     Scale Curve=1.5
    )
    Index 3 Param 1=&FloatFromCurve(
     FloatCurve=DataInterface() {propertyValues {Curve {keys [
         <at 0.5 0 RCIM_Cubic>
         {arriveTangent=1, interpMode=RCIM_Cubic, leaveTangent=1, time=1, value=1}
         {arriveTangent=-1, interpMode=RCIM_Cubic, leaveTangent=-1, time=1.5, value=1}
         {interpMode=RCIM_Cubic, leaveTangent=2, time=2, value=0}
        ]}}}
     CurveIndex=$Emitter.Age
     Scale Curve=&Multiply_Float(
      A=&FN_Select_Float_from_Time_Of_Day(Morning=3, Day=3, Evening=3, Night=3)
      B=&FN_FrontEndFloatModifier(Gameplay=1, Front End=2)
     )
    )
    Index 3 Param 2=&Multiply_Float(
     A=&Lerp_Float(
      A=33
      B=75
      Alpha=&FloatFromCurve(
       FloatCurve=DataInterface() {propertyValues {Curve {keys [
           <at 0.5 0 RCIM_Cubic>
           {arriveTangent=1, interpMode=RCIM_Cubic, leaveTangent=1, time=1, value=1}
           {arriveTangent=-1, interpMode=RCIM_Cubic, leaveTangent=-1, time=1.5, value=1}
           <at 2 0 RCIM_Cubic>
          ]}}}
       CurveIndex=$Emitter.Age
      )
     )
     B=&FN_FrontEndFloatModifier(Gameplay=1, Front End=1.25)
    )
    Module.Param0WriteEnabled=true
    Module.Param0WriteEnabled2=true
    Module.Param0WriteEnabled3=true
    Module.Param1WriteEnabled=false
    Module.Param1WriteEnabled2=true
    Module.Param1WriteEnabled3=true
    Module.Param2WriteEnabled=true
    Module.Param2WriteEnabled2=true
    Module.Param2WriteEnabled3=true
    Module.Param3WriteEnabled=true
    Module.Param3WriteEnabled2=true
    Module.Param3WriteEnabled3=true
    Write Parameter Index 2=true
    Write Parameter Index 3=true
   )
   &ParticleState()
   &ScaleSpriteSize(Uniform Curve Sprite Scale=DataInterface() {propertyValues {Curve {keys [
        {arriveTangent=1.6448, interpMode=RCIM_Cubic, leaveTangent=1.6448, tangentMode=RCTM_User, time=0, value=0.4}
        {arriveTangent=0.0224, interpMode=RCIM_Cubic, leaveTangent=0.0224, tangentMode=RCTM_User, time=1, value=1}
       ]}}})
   &ScaleColor(
    Scale Alpha=&FloatFromCurve(
     FloatCurve=DataInterface() {propertyValues {Curve {keys [
         {interpMode=RCIM_Cubic, tangentMode=RCTM_User, time=0, value=0}
         {arriveTangent=0.0617, interpMode=RCIM_Cubic, leaveTangent=0.0617, tangentMode=RCTM_User, time=0.2, value=1}
         {arriveTangent=0.0658, interpMode=RCIM_Cubic, leaveTangent=0.0658, tangentMode=RCTM_User, time=0.7, value=1}
         {tangentMode=RCTM_User, time=1, value=0}
        ]}}}
     Scale Curve=&Multiply_Float(A=1, B=$System.Timeline001.Time)
    )
    Scale RGB=&MakeVectorFromLinearColorRGB(Color=&MultiplyLinearColor(
      Linear Color=&RandomRangeLinearColor(Minimum=<rgba 0 1 0.15 1>, Maximum=<rgba 0 0.75 0.1 1>)
      Scale Factor=&MakeVector4_FromLinearColor(LinearColor=&LerpLinearColors(StartColor=<rgba 1 1 1 1>, EndColor=$System.TODColor, LerpFactor=0.1))
     ))
    ScaleA=true
    ScaleRGB=true
   )
   &VortexForce(UseOriginOffset=false, Vortex Force Amount=&Lerp_Float(A=-800, B=-300, Alpha=$System.Timeline002.Float))
   &Drag(Drag=10, Use Linear Drag=true, Use Rotational Drag=true)
   &AccelerationForce(Acceleration=&Multiply_VectorByFloat(Vector=<1 1 1>, Float=1))
   &SolveForcesAndVelocity(Clamp Velocity=false, Limit Acceleration=false)
  ]
  rendererClasses [&NiagaraSpriteRendererProperties]
  renderers [
   {rendererClass=&NiagaraSpriteRendererProperties}
   {rendererClass=&NiagaraSpriteRendererProperties, rendererIndex=1}
  ]
 }
 Spiral {
  emitterUpdateScript [
   &EmitterState(Life Cycle Mode=::NewEnumerator1, Loop Behavior=::NewEnumerator1, Loop Duration Mode=::NewEnumerator1, UseLoopDelay=false)
   &SpawnBurst_Instantaneous(Spawn Count=3, Use Spawn Probability=false, UseLoopCountLimit=false)
  ]
  particleSpawnScript [
   &PartitionParticles(Number Of Partitions=4)
   &InitializeParticle(
    Color=<rgba 0 0.3 0.05 1>
    Lifetime=5
    Mesh Scale Mode=::NewEnumerator3
    Mesh Uniform Scale=&Multiply_Float(
     A=1.05
     B=&SelectFloatFromArray(
      Direct Array Index=$Particles.PartitionParticles.Partition
      Float Selection Array=DataInterface() {propertyValues {FloatData [1.4,1.45,1.5]}}
     )
    )
    Position Offset=<0 0 -5>
    UsePositionOffset=true
    Write Material Random=true
   )
   /SetVariables(Particles.RandomTimeFloat:NiagaraFloat=&RandomRangeFloat(Minimum=0.4, Maximum=1)) {isSetParameters=true}
   &InitialMeshOrientation(Mesh Orientation Mode=::NewEnumerator2, Rotation=<0 0 -0.1>, Use Rotation Vector=true)
   &CameraOffset(Camera Offset Amount=10)
   /SetVariables(Particles.LightRenderVis:NiagaraInt32=&SelectIntFromArray(
     Direct Array Index=&ReturnExecIndex()
     Int Selection Array=DataInterface() {propertyValues {IntData [0,1,1]}}
    )) {isSetParameters=true}
  ]
  particleUpdateScript [
   &ParticleState(Kill Particles When Lifetime Has Elapsed=false, Let Infinitely Lived Particles Die When Emitter Deactivates=true)
   &DynamicMaterialParameters(
    Index 0 Param 1=&Multiply_Float(A=$Engine.RealTime, B=0.6)
    Index 0 Param 2=&Multiply_Float(A=$Engine.RealTime, B=0.6)
    Index 0 Param 3=&SelectFloatFromArray(Direct Array Index=$Particles.PartitionParticles.Partition, Float Selection Array=DataInterface() {propertyValues {FloatData [0,0.2,0.4]}})
    Index 0 Param 4=&RandomRangeFloat(Minimum=0.2, Maximum=0.4)
    Index 1 Param 1=&Lerp_Float(
     A=0
     B=1
     Alpha=&FloatFromCurve(
      FloatCurve=DataInterface() {propertyValues {Curve {keys [{arriveTangent=1.0694, interpMode=RCIM_Cubic, leaveTangent=1.0694, tangentMode=RCTM_User, time=0, value=0}, <at 3 1 RCIM_Cubic>]}}}
      CurveIndex=$Emitter.Age
     )
    )
    Index 2 Param 1=&FloatFromCurve(
     FloatCurve=DataInterface() {propertyValues {Curve {keys [
         <at 0.5 0 RCIM_Cubic>
         {arriveTangent=1, interpMode=RCIM_Cubic, leaveTangent=1, time=1, value=1}
         {arriveTangent=-1, interpMode=RCIM_Cubic, leaveTangent=-1, time=1.5, value=1}
         <at 2 0 RCIM_Cubic>
        ]}}}
     CurveIndex=$Emitter.Age
     Scale Curve=&FN_Select_Float_from_Time_Of_Day(Morning=16, Day=16, Evening=16, Night=10)
    )
    Index 2 Param 2=&Multiply_Float(
     A=&Lerp_Float(
      A=33
      B=66
      Alpha=&FloatFromCurve(
       FloatCurve=DataInterface() {propertyValues {Curve {keys [
           <at 0.5 0 RCIM_Cubic>
           {arriveTangent=1, interpMode=RCIM_Cubic, leaveTangent=1, time=1, value=1}
           {arriveTangent=-1, interpMode=RCIM_Cubic, leaveTangent=-1, time=1.5, value=1}
           <at 2 0 RCIM_Cubic>
          ]}}}
       CurveIndex=$Emitter.Age
      )
     )
     B=&FN_FrontEndFloatModifier(Gameplay=1, Front End=1.25)
    )
    Module.Index0Param0Write=true
    Module.Index0Param1Write=true
    Module.Index0Param2Write=true
    Module.Index0Param3Write=true
    Module.Index1Param0Write=true
    Module.Index1Param1Write=true
    Module.Index1Param2Write=true
    Module.Index1Param3Write=true
    Module.Index2Param0Write=true
    Module.Index2Param1Write=true
    Module.Index2Param2Write=true
    Module.Index2Param3Write=true
    Write Parameter Index 1=true
    Write Parameter Index 2=true
   )
   &ScaleColor(Scale Alpha=&Lerp_Float(A=0, B=2, Alpha=$System.Timeline001.Time), ScaleA=true, ScaleRGB=true)
   &FN_Light_Attributes(
    Color=<rgba 0.0425 0.5156 0 1>
    Create Color Attribute=true
    Create Exponent Attribute=true
    Create Position Attribute=true
    Create Radius Attribute=true
    Exponent=2
    Position=<0 0 -5>
    Radius=&FN_FrontEndFloatModifier(Gameplay=100, Front End=0)
    Scale Color=&Lerp_Float(
     A=0
     B=5
     Alpha=&FloatFromCurve(
      FloatCurve=DataInterface() {propertyValues {Curve {keys [{tangentMode=RCTM_User, time=0.5, value=0}, {tangentMode=RCTM_User, time=1.5, value=1}]}}}
      CurveIndex=$Emitter.Age
      Scale Curve=&FN_FrontEndFloatModifier(Gameplay=5, Front End=0)
     )
    )
   )
  ]
  rendererClasses [&NiagaraMeshRendererProperties, &NiagaraLightRendererProperties]
  renderers [
   {rendererClass=&NiagaraMeshRendererProperties, rendererIndex=0}
   {rendererClass=&NiagaraLightRendererProperties, rendererIndex=1}
   {rendererClass=&NiagaraMeshRendererProperties, rendererIndex=2}
  ]
 }
 Swirl {
  emitterUpdateScript [
   &EmitterState(
    Inactive Response=::NewEnumerator1
    Life Cycle Mode=::NewEnumerator1
    Loop Behavior=::NewEnumerator1
    Loop Duration Mode=::NewEnumerator1
    UseLoopDelay=false
   )
   &SpawnBurst_Instantaneous(
    Spawn Count=&SelectIntFromArray(Direct Array Index=$Engine.QualityLevel, Int Selection Array=DataInterface() {propertyValues {IntData [6,6,8,10,10]}})
    Use Spawn Probability=false
    UseLoopCountLimit=false
   )
  ]
  particleSpawnScript [
   &PartitionParticles(
    Particles Per Partition:NiagaraInt32=&SelectIntFromArray(Direct Array Index=$Engine.QualityLevel, Int Selection Array=DataInterface() {propertyValues {IntData [4,4,5,6,6]}})
    Partition Mode:ENiagara_PartitionMode=::NewEnumerator1
   )
   &InitializeParticle(
    Color=<rgba 1 1 1 2>
    Lifetime Min=1.5
    Lifetime Mode=::NewEnumerator1
    Mesh Scale Mode=::NewEnumerator3
    Mesh Uniform Scale=&MaskFloatBySpawnGroup(
     Masked Float=&FloatFromCurve(
      FloatCurve=DataInterface() {propertyValues {Curve {keys [<at 0 0.5 RCIM_Cubic>, {arriveTangent=0.961, interpMode=RCIM_Cubic, leaveTangent=0.961, tangentMode=RCTM_User, time=1, value=1}]}}}
      CurveIndex=$Particles.PartitionParticles.Normalized Index In Partition
      Scale Curve=0.8
     )
     Passthrough Float=&FloatFromCurve(
      FloatCurve=DataInterface() {propertyValues {Curve {keys [{tangentMode=RCTM_User, time=0, value=1}, {tangentMode=RCTM_User, time=1, value=0.2}]}}}
      CurveIndex=$Particles.PartitionParticles.Normalized Index In Partition
     )
     Spawn Group Mask=$Particles.PartitionParticles.Partition
    )
    Position Offset=&MaskVectorBySpawnGroup(
     Masked Vector=&Lerp_Vector(
      A=<0 0 10>
      B=<0 0 45>
      Alpha=&FloatFromCurve(
       FloatCurve=DataInterface() {propertyValues {Curve {keys [{tangentMode=RCTM_User, time=0, value=0}, {tangentMode=RCTM_User, time=1, value=1}]}}}
       CurveIndex=$Particles.PartitionParticles.Normalized Index In Partition
      )
     )
     Passthrough Vector=&Lerp_Vector(
      A=<0 0 -20>
      B=<0 0 40>
      Alpha=&FloatFromCurve(
       FloatCurve=DataInterface() {propertyValues {Curve {keys [
           {interpMode=RCIM_Cubic, tangentMode=RCTM_User, time=0, value=0.5}
           {interpMode=RCIM_Cubic, tangentMode=RCTM_User, time=0.5, value=1}
           {tangentMode=RCTM_User, time=1, value=0.5}
          ]}}}
       CurveIndex=$Particles.PartitionParticles.Normalized Index In Partition
       Scale Curve=0.8
      )
     )
     Spawn Group Mask=$Particles.PartitionParticles.Partition
    )
    UsePositionOffset=true
    Write Material Random=true
   )
   &InitialMeshOrientation(
    Mesh Axis To Orient=<0 0 1>
    Mesh Orientation Mode=::NewEnumerator2
    Orientation Vector=<0 1 0>
    Rotation=&RandomRangeVector(Minimum=<0 -1 0.075>, Maximum=<0 1 -0.075>)
    Use Rotation Vector=true
   )
   /SetVariables(Particles.RandomTimeFloat:NiagaraFloat=&RandomRangeFloat(Minimum=0.8, Maximum=1)) {isSetParameters=true}
  ]
  particleUpdateScript [
   &ParticleState(
    Kill Particles When Lifetime Has Elapsed=false
    Let Infinitely Lived Particles Die When Emitter Deactivates=true
    Loop Particles Lifetime=true
   )
   &UpdateMeshOrientation(Rotation Rate=&RandomRangeFloat(Minimum=0, Maximum=0), Rotation Vector=<0 1 0>)
   &ScaleColor(
    Scale Alpha=&MaskFloatBySpawnGroup(
     Masked Float=&FloatFromCurve(
      FloatCurve=DataInterface() {propertyValues {Curve {keys [<at 0 0 RCIM_Cubic>, {arriveTangent=0.2123, interpMode=RCIM_Cubic, leaveTangent=0.2123, tangentMode=RCTM_User, time=1, value=1.2}]}}}
      CurveIndex=$Particles.PartitionParticles.Normalized Index In Partition
      Scale Curve=&Lerp_Float(A=0, B=0.4, Alpha=$System.Timeline001.Time)
     )
     Passthrough Float=&Multiply_Float(
      A=&Lerp_Float(A=1, B=0, Alpha=$System.Timeline002.Time)
      B=&Lerp_Float(A=0, B=0.4, Alpha=$System.Timeline001.Time)
     )
     Spawn Group Mask=$Particles.PartitionParticles.Partition
    )
    Scale RGB=&MakeVectorFromLinearColorRGB(Color=&MultiplyLinearColor(
      Linear Color=<rgba 0 0.4 0 1>
      Scale Factor=&MakeVector4_FromLinearColor(LinearColor=&LerpLinearColors(StartColor=<rgba 1 1 1 1>, EndColor=$System.TODColor, LerpFactor=0.1))
     ))
    ScaleA=true
    ScaleRGB=true
   )
   &DynamicMaterialParameters(
    Index 0 Param 1=&Multiply_Float(A=$Engine.RealTime, B=$Particles.RandomTimeFloat)
    Index 0 Param 2=&Multiply_Float(A=$Engine.RealTime, B=$Particles.RandomTimeFloat)
    Module.Index0Param0Write=true
    Module.Index0Param1Write=true
    Module.Index0Param2Write=true
    Module.Index0Param3Write=true
   )
  ]
  rendererClasses [&NiagaraMeshRendererProperties]
  renderers [
   {rendererClass=&NiagaraMeshRendererProperties, rendererIndex=0}
   {rendererClass=&NiagaraMeshRendererProperties, rendererIndex=1}
  ]
 }
 WispySmoke {
  emitterUpdateScript [
   &EmitterState(Life Cycle Mode=::NewEnumerator1, Loop Duration=1, UseLoopDelay=false)
   &SpawnRate(SpawnRate=8, Use Spawn Probability=false)
  ]
  particleSpawnScript [
   &InitializeParticle(
    Lifetime=2
    Sprite Rotation Mode=::NewEnumerator2
    Sprite Size Mode=::NewEnumerator3
    Uniform Sprite Size=100
    UsePositionOffset=false
    Write Material Random=true
   )
   &ShapeLocation(
    Cylinder Distribution=::NewEnumerator2
    Cylinder Height=1
    Cylinder Radius=&Lerp_Float(A=30, B=40, Alpha=$System.Timeline002.Float)
    MaskSpawnGroup=false
    Rotation Angle=180
    Rotation Axis=<0 0 1>
    Rotation Mode=::NewEnumerator1
    Shape Primitive=::NewEnumerator1
    U Position=&RandomRangeFloat(Minimum=0, Maximum=1)
    UseOwnerScale=true
   )
  ]
  particleUpdateScript [
   &InheritSourceMovement(
    Applied Position Scale=&Multiply_VectorByFloat(
     Vector=<1 1 1>
     Float=&Lerp_Float(
      A=&FloatFromCurve(FloatCurve=DataInterface() {propertyValues {Curve {keys [{arriveTangent=-1.0658, interpMode=RCIM_Cubic, leaveTangent=-1.0658, tangentMode=RCTM_User, time=0.1, value=1}, <at 1 0.4 RCIM_Cubic>]}}})
      B=&FloatFromCurve(FloatCurve=DataInterface() {propertyValues {Curve {keys [{arriveTangent=-0.0965, interpMode=RCIM_Cubic, leaveTangent=-0.0965, tangentMode=RCTM_User, time=0.3, value=1}, <at 1 0.7 RCIM_Cubic>]}}})
      Alpha=&MakeCustomFloatFromBool(
       Bool=&SetBoolByFloatComparison(A=&VectorLength(Vector=$Engine.Owner.Velocity), B=400)
       True Float=1
       False Float=0
      )
     )
    )
    ApplyForce=false
    ApplyPosition=true
    ApplyPositionFalloff=false
   )
   &GravityForce(Gravity=<0 0 100>)
   &SolveForcesAndVelocity(Clamp Velocity=false, Limit Acceleration=false)
   &ParticleState()
   &ScaleSpriteSize(Uniform Curve Sprite Scale=DataInterface() {propertyValues {Curve {keys [<at 0 0.5 RCIM_Cubic>, {arriveTangent=0.5164, interpMode=RCIM_Cubic, leaveTangent=0.5164, tangentMode=RCTM_User, time=1, value=1}]}}})
   &SubUVAnimation(
    End Frame Range Override=32
    "Sprite Renderer "=DataInterface() {propertyValues {SpriteRenderer=&self:WispySmoke_2.NiagaraSpriteRendererProperties_0}}
    Start Frame Range Override=16
    UseEndFrame=true
    UseStartFrame=true
   )
   &ScaleColor(
    Scale Alpha=&FloatFromCurve(FloatCurve=DataInterface() {propertyValues {Curve {keys [
         {interpMode=RCIM_Cubic, tangentMode=RCTM_User, time=0, value=0}
         {interpMode=RCIM_Cubic, tangentMode=RCTM_User, time=0.5, value=1}
         {tangentMode=RCTM_User, time=1, value=0}
        ]}}})
    Scale RGB=&MakeVectorFromLinearColorRGB(Color=&MultiplyLinearColor(
      Linear Color=<rgba 0 0.7 0.15 1>
      Scale Factor=&MakeVector4_FromLinearColor(LinearColor=&LerpLinearColors(StartColor=<rgba 1 1 1 1>, EndColor=$System.TODColor, LerpFactor=0.5))
     ))
    ScaleA=true
    ScaleRGB=true
   )
   &VortexForce(UseOriginOffset=false, Vortex Force Amount=&Lerp_Float(A=-800, B=-300, Alpha=$System.Timeline002.Float))
   &Drag(Drag=12, Use Linear Drag=true, Use Rotational Drag=true) {module=Drag001}
   &SolveForcesAndVelocity(Clamp Velocity=false, Limit Acceleration=false) {module=SolveForcesAndVelocity001}
  ]
  rendererClasses [&NiagaraSpriteRendererProperties]
  renderers [{rendererClass=&NiagaraSpriteRendererProperties}]
 }
}
summary {
 systemName=NS_Soilblend_Emote_Smoke
 emitters [
  {emitterName=Swirl, rendererClasses [&NiagaraMeshRendererProperties]}
  {emitterName=WispySmoke, rendererClasses [&NiagaraSpriteRendererProperties]}
  {emitterName=Spiral, rendererClasses [&NiagaraMeshRendererProperties, &NiagaraLightRendererProperties]}
  {emitterName=Smoke001, rendererClasses [&NiagaraSpriteRendererProperties]}
 ]
 userVariables [
  User.bIsFrontEnd {type=NiagaraBool}
  User.bIsFrontEndPreview {type=NiagaraBool}
 ]
}
